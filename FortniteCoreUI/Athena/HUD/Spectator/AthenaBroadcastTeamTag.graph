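# Spectator broadcast team tag widget showing a squad-size icon, a "Team" label and an optional session rank number.
# UpdateMaxSquadSize looks up a duo or trio icon in a map, falling back to the default squad icon, and PreConstruct previews it at design time.
# UpdateSessionRank and ToggleAllowSessionRank show the rank only when allowed and non-negative.
# ToggleDisplaySessionRank collapses or reveals the rank container and switches the tag width to match.
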WidgetBlueprint /FortniteCoreUI/Athena/HUD/Spectator/AthenaBroadcastTeamTag
graphs [
 EventGraph {nodes [
   n1 = Event UpdateMaxSquadSize!() -> (OutputDelegate, then, MaxSquadSize:int=0) {is=title K2Node_Event}
   n2 = Find(
    TargetMap:Map of Integers to Texture 2D Soft Object References=Get TeamIcons() -> (TeamIcons:Map of Integers to Texture 2D Soft Object References)
    Key:"Integer (by ref)"=@n1.MaxSquadSize
   ) -> (Value:Texture 2D Soft Object Reference)
   > SetBrushFromSoftTexture!(
    execute
    self:Image Object Reference=Get TeamIcon() -> (TeamIcon:Image Object Reference)
    SoftTexture:Texture 2D Soft Object Reference=Select(
     Option 0:Texture 2D Soft Object Reference=Get DefaultTeamIcon() -> (DefaultTeamIcon:Texture 2D Soft Object Reference)
     Option 1:Texture 2D Soft Object Reference=@n2
     Index=@n2.ReturnValue
    ) -> (ReturnValue:Texture 2D Soft Object Reference)
    bMatchSize:bool=false
   ) -> (then)
   n3 = Event PreConstruct()
   > Branch(Condition=@n3.IsDesignTime)
   > UpdateMaxSquadSize!(
    execute
    self:Athena Spectator Team Tag Base Object Reference
    MaxSquadSize:int=Get DesignTime_MaxSquadSize() -> (DesignTime_MaxSquadSize:int=0)
   ) -> (then)
  ]}
 ToggleAllowSessionRank {nodes [
   n1 = ToggleAllowSessionRank!() -> (then, Allow:bool=~) {is=title K2Node_FunctionEntry}
   > n2 = Set AllowSessionRank(AllowSessionRank:bool=@n1) -> (Output_Get:bool)
   n3 = Get InSessionRank() -> (InSessionRank:int=0)
   > n4 = Set ShouldDisplaySessionRank(ShouldDisplaySessionRank:bool=AND Boolean(A=@n2, B="integer >= integer"!(A:int=@n3, B:int=0) -> (ReturnValue:bool) {is=title K2Node_CallFunction})) -> (Output_Get:bool)
   > ToggleDisplaySessionRank!(execute, self, Display:bool=@n4) -> (then)
  ]}
 ToggleDisplaySessionRank {nodes [
   n1 = ToggleDisplaySessionRank!() -> (then, Display:bool=~) {is=title K2Node_FunctionEntry}
   > SetWidthOverride(
    self=Get ContainerSizeBox() -> (ContainerSizeBox:Size Box Object Reference)
    InWidthOverride=Select(Option 0:double=135, Option 1:double=180, Index=@n1) -> (ReturnValue:double)
   )
   > SetVisibility(
    self=Get SessionRankContainer() -> (SessionRankContainer:Horizontal Box Object Reference)
    InVisibility=Select(Option 0:ESlateVisibility Enum=Collapsed, Option 1:ESlateVisibility Enum=Visible, Index=@n1) -> (ReturnValue:ESlateVisibility Enum)
   )
  ]}
 UpdateSessionRank {nodes [
   n1 = UpdateSessionRank!() -> (then, InSessionRank:int) {is=title K2Node_FunctionEntry}
   > n2 = Set InSessionRank(InSessionRank:int=@n1) -> (Output_Get:int=0)
   > n3 = Set ShouldDisplaySessionRank(ShouldDisplaySessionRank=AND Boolean(A="integer >= integer"(A=@n2, B=0), B=Get AllowSessionRank() -> (AllowSessionRank:bool)))
   > ToggleDisplaySessionRank(Display=@n3) {is=title K2Node_CallFunction}
   > Branch(Condition=@n3)
   n4 = Get InSessionRank() -> (InSessionRank=0)
   > "SetText (Text)"(self=Get SessionRank() -> (SessionRank:Common Text Object Reference), InText="To Text (String)"(InString="To String (Integer)"(InInt=@n4)))
  ]}
]
variables [
 AllowSessionRank {type=boolean}
 DefaultTeamIcon {type=Texture2D, default=/Game/UI/Foundation/Textures/Icons/Social/T_Icon_Matchmaking_Squad_64}
 DesignTime_MaxSquadSize {type=integer, default=4}
 InSessionRank {type=integer, default=-1}
 ShouldDisplaySessionRank {type=boolean}
 TeamIcons {default {
   2=/Game/UI/Foundation/Textures/Icons/Social/T_Icon_Matchmaking_Duo_64
   3=/Game/UI/Foundation/Textures/Icons/Social/T_Icon_Matchmaking_Trios_64
  }}
]
widgets {
 info {inheritedWidgetCount=0, parentClass=/Script/FortniteUI.AthenaSpectatorTeamTagBase, rootWidgetClass=&SizeBox, widgetCount=9}
 widgets [
  ContainerSizeBox {
   widgetClassPath=&SizeBox
   bIsVariable=true
   widget=ContainerSizeBox
   properties {bOverride_HeightOverride=true, bOverride_WidthOverride=true, heightOverride=48, widthOverride=180}
  }
  CommonBorder_0 {
   widgetClassPath=&CommonBorder
   parent=ContainerSizeBox
   widget=CommonBorder_0
   properties {
    verticalAlignment=VAlign_Center
    background {
     bIsDynamicallyLoaded=false
     drawAs=Box
     imageType=NoImage
     mirroring=NoMirror
     resourceName=None
     resourceObject=/Game/Athena/HUD/Spectator/Art/M_UI_SkewBox_Broadcast_PlayersRemaining
     tiling=NoTile
     imageSize=<32 32>
     margin=<ltrb 0 0 0 0>
     outlineSettings {
      bUseBrushTransparency=false
      roundingType=HalfHeightRadius
      width=0
      color {colorUseRule=UseColor_Specified, specifiedColor=<rgba 0 0 0 0>}
      cornerRadii=<0 0 0 0>
     }
     tintColor {colorUseRule=UseColor_Specified, specifiedColor=<rgba 1 1 1 1>}
     uVRegion {bIsValid=false, max=<0 0>, min=<0 0>}
    }
    padding=<ltrb 0 0 0 0>
   }
   slot {class=SizeBoxSlot}
  }
  HorizontalBox {
   widgetClassPath=&HorizontalBox
   parent=CommonBorder_0
   widget=HorizontalBox
   slot {class=BorderSlot, properties {verticalAlignment=VAlign_Center, padding=<ltrb 0 0 0 0>}}
  }
  TeamIcon {
   widgetClassPath=&Image
   bIsVariable=true
   parent=HorizontalBox
   widget=TeamIcon
   properties {
    visibility=SelfHitTestInvisible
    brush {
     bIsDynamicallyLoaded=false
     drawAs=Image
     imageType=NoImage
     mirroring=NoMirror
     resourceName=None
     resourceObject=&T_Icon_Matchmaking_Squad_64
     tiling=NoTile
     imageSize=<30 30>
     margin=<ltrb 0 0 0 0>
     outlineSettings {
      bUseBrushTransparency=false
      roundingType=HalfHeightRadius
      width=0
      color {colorUseRule=UseColor_Specified, specifiedColor=<rgba 0 0 0 0>}
      cornerRadii=<0 0 0 0>
     }
     tintColor {colorUseRule=UseColor_Specified, specifiedColor=<rgba 1 1 1 1>}
     uVRegion {bIsValid=false, max=<0 0>, min=<0 0>}
    }
   }
   slot {class=HorizontalBoxSlot, properties {verticalAlignment=VAlign_Center, padding=<ltrb 10 0 5 0>}}
  }
  SizeBox_1 {
   widgetClassPath=&SizeBox
   parent=HorizontalBox
   widget=SizeBox_1
   properties {bOverride_WidthOverride=true, maxDesiredWidth=59, minDesiredWidth=59, widthOverride=59}
   slot {class=HorizontalBoxSlot, properties {verticalAlignment=VAlign_Center, padding=<ltrb 0 0 15 0>, size {sizeRule=Automatic, value=0.3}}}
  }
  CommonTextBlock_100 {
   widgetClassPath=&CommonTextBlock
   parent=SizeBox_1
   widget=CommonTextBlock_100
   properties {
    justification=Center
    scrollStyle=/FortUILibrary/Styles/Typography/Foundation/TextScrollStyle-NoFade.TextScrollStyle-NoFade_C
    text=Team
    textTransformPolicy=ToUpper
    font {
     bForceMonospaced=false
     bMaterialIsStencil=false
     fontMaterial=null
     fontObject=&BurbankSmall
     letterSpacing=0
     monospacedWidth=1
     size=16
     skewAmount=0
     typefaceFontName=Black
     outlineSettings {
      bApplyOutlineToDropShadows=false
      bMiteredCorners=false
      bSeparateFillAlpha=false
      outlineMaterial=null
      outlineSize=0
      outlineColor=<rgba 0 0 0 1>
     }
    }
    shadowOffset=<0 0>
   }
   slot {class=SizeBoxSlot}
  }
  SessionRankContainer {
   widgetClassPath=&HorizontalBox
   bIsVariable=true
   parent=HorizontalBox
   widget=SessionRankContainer
   slot {class=HorizontalBoxSlot}
  }
  SessionRankIndicator {
   widgetClassPath=&CommonTextBlock
   parent=SessionRankContainer
   widget=SessionRankIndicator
   properties {
    style=/FortUILibrary/Styles/Typography/Foundation/TextStyle-BurbankSmall-XS-White.TextStyle-BurbankSmall-XS-White_C
    text="##"
    font {
     bForceMonospaced=false
     bMaterialIsStencil=false
     fontMaterial=null
     fontObject=&BurbankSmall
     letterSpacing=0
     monospacedWidth=1
     size=13
     skewAmount=0
     typefaceFontName=Black
     outlineSettings {
      bApplyOutlineToDropShadows=false
      bMiteredCorners=false
      bSeparateFillAlpha=false
      outlineMaterial=null
      outlineSize=0
      outlineColor=<rgba 0 0 0 1>
     }
    }
    shadowOffset=<0 0>
   }
   slot {class=HorizontalBoxSlot, properties {verticalAlignment=VAlign_Bottom, padding=<ltrb 0 0 0 2>}}
  }
  SessionRank {
   widgetClassPath=&CommonTextBlock
   bIsVariable=true
   parent=SessionRankContainer
   widget=SessionRank
   properties {
    style=/FortUILibrary/Styles/Typography/Foundation/TextStyle-BurbankSmall-S-White.TextStyle-BurbankSmall-S-White_C
    text=99
    font {
     bForceMonospaced=false
     bMaterialIsStencil=false
     fontMaterial=null
     fontObject=&BurbankSmall
     letterSpacing=0
     monospacedWidth=1
     size=17
     skewAmount=0
     typefaceFontName=Black
     outlineSettings {
      bApplyOutlineToDropShadows=false
      bMiteredCorners=false
      bSeparateFillAlpha=false
      outlineMaterial=null
      outlineSize=0
      outlineColor=<rgba 0 0 0 1>
     }
    }
    shadowOffset=<0 0>
   }
   slot {class=HorizontalBoxSlot}
  }
 ]
}
