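# UI material that draws a healing indicator icon, reusing the squad DBNO icon's blur and cross textures.
# The blur texture is tinted green, then blended toward the cross texture using the cross's own RGB as the mask.
# The larger of the two texture alphas provides the combined translucent shape, with no animation or parameters.
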Material /Game/Athena/HUD/Art/M_Healing_Indicator
graph {
 properties {blendMode=BLEND_Translucent, materialDomain=MD_UI, pixelDepthOffsetMode=PDOM_Legacy, usedShadingModels=Default Lit}
 nodes [
  n1 = TextureSample() {texture=&/squad_DBNOIcon_Blur}
  n2 = TextureSample() {texture=&/squad_DBNOIcon_Cross}
  LinearInterpolate(A=Multiply(A=Constant3Vector() {constant=<rgba 0 0.5457 0.0115 1>}, B=@n1.RGB), B=@n2.RGB, Alpha=@n2.RGB)
  Max(A=@n2.A, B=@n1.A)
 ]
}
