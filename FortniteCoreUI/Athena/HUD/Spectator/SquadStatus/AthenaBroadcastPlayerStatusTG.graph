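# Spectator HUD row for one squad player showing name, squad ID, eliminations, health, shield and downed or revive state.
# Player data events drive the text, bars and icons, while followed-player and picture-in-picture modes switch backgrounds, fonts and sizing.
# Downed and eliminated animations differ for followed player and teammates, with colors restored from desired values derived from minigame team.
# Clicking a teammate row makes that player the followed one.
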WidgetBlueprint /FortniteCoreUI/Athena/HUD/Spectator/SquadStatus/AthenaBroadcastPlayerStatusTG
animations [
 SpectatedDowned {
  binding_count=2
  length_seconds=1
  object=&self:SpectatedDowned
  sequence {bindings [
    FollowedPlayerGradient {
     id=#1
     possessed_class=BasicGradientFill_C
     tracks [
      {
       type=MovieSceneColorTrack
       display_name=Color 1
       sections [{
         type=MovieSceneColorSection
         end_frame=19
         start_frame=0
         channels [
          Color.R {type=MovieSceneScriptingFloatChannel, keys [[0,0.0044],[2,0.8148],[17.9997,0.8148],[19.999,0.0044]]}
          Color.G {type=MovieSceneScriptingFloatChannel, keys [[0,0.1022],[2,0.0369],[17.9997,0.0369],[19.999,0.1022]]}
          Color.B {type=MovieSceneScriptingFloatChannel, keys [[0,0.4793],[2,0.0307],[17.9997,0.0307],[19.999,0.4793]]}
          Color.A {type=MovieSceneScriptingFloatChannel, const [1,4]}
         ]
        }]
      }
      {
       type=MovieSceneColorTrack
       display_name=Color 2
       sections [{
         type=MovieSceneColorSection
         end_frame=19
         start_frame=0
         channels [
          Color.R {type=MovieSceneScriptingFloatChannel, keys [[0,0.0523],[1.9997,0.5647],[17.9997,0.5647],[19.999,0.0523]]}
          Color.G {type=MovieSceneScriptingFloatChannel, keys [[0,0.1526],[1.9997,0.0075],[17.9997,0.0075],[19.999,0.1526]]}
          Color.B {type=MovieSceneScriptingFloatChannel, keys [[0,0.5647],[1.9997,0.1912],[17.9997,0.1912],[19.999,0.5647]]}
          Color.A {type=MovieSceneScriptingFloatChannel, const [1,2]}
         ]
        }]
      }
      {type=MovieSceneEventTrack, sections [{type=MovieSceneEventRepeaterSection, end_frame=19, start_frame=0}]}
     ]
    }
    FollowedPlayerElimBackground {
     id=#2
     possessed_class=CommonBorder
     tracks [{
       type=MovieSceneWidgetMaterialTrack
       sections [{
         type=MovieSceneParameterSection
         end_frame=20
         start_frame=0
         channels [
          RGBR {type=MovieSceneScriptingFloatChannel, keys [[0,0.0423],[1.9997,0.9412],[18,0.9412],[19.9997,0.0423]]}
          RGBG {type=MovieSceneScriptingFloatChannel, keys [[0,0.3325],[1.9997,0.098],[18,0.098],[19.9997,0.3325]]}
          RGBB {type=MovieSceneScriptingFloatChannel, keys [[0,0.8879],[1.9997,0.2275],[18,0.2275],[19.9997,0.8879]]}
          RGBA {type=MovieSceneScriptingFloatChannel, const [1,2]}
         ]
        }]
      }]
    }
   ]}
 }
 TeammateDowned {
  binding_count=3
  length_seconds=1
  object=&self:TeammateDowned
  sequence {bindings [
    SquadMateBackground {
     id=#2
     display_name=CommonBorder_47
     possessed_class=CommonBorder
     tracks [{
       type=MovieSceneColorTrack
       display_name=Brush Color
       sections [{
         type=MovieSceneColorSection
         end_frame=20
         start_frame=0
         channels [
          Color.R {type=MovieSceneScriptingFloatChannel, keys [[0,0],[2,0.8148],[18,0.8148],[20,0]]}
          Color.G {type=MovieSceneScriptingFloatChannel, keys [[0,0],[2,0.0369],[18,0.0369],[20,0]]}
          Color.B {type=MovieSceneScriptingFloatChannel, keys [[0,0],[2,0.0307],[18,0.0307],[20,0]]}
          Color.A {type=MovieSceneScriptingFloatChannel, const [0.298,2]}
         ]
        }]
      }]
    }
    ShieldBar {
     id=#3
     possessed_class=AthenaSpectatorHitPointBar_C
     tracks [{
       type=MovieSceneByteTrack
       display_name=Visibility
       sections [{type=MovieSceneByteSection, end_frame=20, start_frame=0, channels [None {type=MovieSceneScriptingByteChannel, keys [[0,4],[2,1],[20,1]]}]}]
      }]
    }
    VerticalBox_42 {
     id=#4
     display_name=HitPointElimIconOverlay
     possessed_class=VerticalBox
     tracks [{
       type=MovieScene2DTransformTrack
       sections [{
         type=MovieScene2DTransformSection
         end_frame=20
         start_frame=0
         channels [
          Translation.X {type=MovieSceneScriptingFloatChannel, const [0,1]}
          Translation.Y {type=MovieSceneScriptingFloatChannel, const [0,1]}
          Angle {type=MovieSceneScriptingFloatChannel, const [0,1]}
          Scale.X {type=MovieSceneScriptingFloatChannel, thinned_from=7, keys [[0,1],[2,1.25],[4,1],[6,1.25],[18,1.25],[20,1]]}
          Scale.Y {type=MovieSceneScriptingFloatChannel, keys [[0,1],[2,1.25],[4,1],[5.9997,1.25],[18,1.25],[20,1]]}
          Shear.X {type=MovieSceneScriptingFloatChannel}
          Shear.Y {type=MovieSceneScriptingFloatChannel}
         ]
        }]
      }]
    }
   ]}
 }
 SpectatedEliminated {
  binding_count=2
  length_seconds=1.1
  object=&self:SpectatedEliminated
  sequence {bindings [
    FollowedPlayerGradient {
     id=#1
     possessed_class=BasicGradientFill_C
     tracks [
      {
       type=MovieSceneColorTrack
       display_name=Color 1
       sections [{
         type=MovieSceneColorSection
         end_frame=22
         start_frame=0
         channels [
          Color.R {type=MovieSceneScriptingFloatChannel, keys [[0,0.0044],[2,0.8148],[18.9997,0.8148],[19.9997,0.1098]]}
          Color.G {type=MovieSceneScriptingFloatChannel, keys [[0,0.1022],[2,0.0369],[18.9997,0.0369],[19.9997,0.1098]]}
          Color.B {type=MovieSceneScriptingFloatChannel, keys [[0,0.4793],[2,0.0307],[18.9997,0.0307],[19.9997,0.1098]]}
          Color.A {type=MovieSceneScriptingFloatChannel, const [1,4]}
         ]
        }]
      }
      {
       type=MovieSceneColorTrack
       display_name=Color 2
       sections [{
         type=MovieSceneColorSection
         end_frame=22
         start_frame=0
         channels [
          Color.R {type=MovieSceneScriptingFloatChannel, keys [[0,0.0523],[1.9997,0.5647],[18.9997,0.5647],[19.999,0.0523],[19.9997,0.0499]]}
          Color.G {type=MovieSceneScriptingFloatChannel, keys [[0,0.1526],[1.9997,0.0075],[18.9997,0.0075],[19.999,0.0523],[19.9997,0.0499]]}
          Color.B {type=MovieSceneScriptingFloatChannel, keys [[0,0.5647],[1.9997,0.1912],[18.9997,0.1912],[19.999,0.0523],[19.9997,0.0499]]}
          Color.A {type=MovieSceneScriptingFloatChannel, const [1,2]}
         ]
        }]
      }
      {type=MovieSceneEventTrack, sections [{type=MovieSceneEventRepeaterSection, end_frame=21, start_frame=0}]}
     ]
    }
    FollowedPlayerElimBackground {
     id=#2
     possessed_class=CommonBorder
     tracks [{
       type=MovieSceneWidgetMaterialTrack
       sections [{
         type=MovieSceneParameterSection
         end_frame=21
         start_frame=0
         channels [
          RGBR {type=MovieSceneScriptingFloatChannel, keys [[0,0.0423],[1.9997,0.9412],[19,0.9412],[19.9997,0.2]]}
          RGBG {type=MovieSceneScriptingFloatChannel, keys [[0,0.3325],[1.9997,0.098],[19,0.098],[19.9997,0.2]]}
          RGBB {type=MovieSceneScriptingFloatChannel, keys [[0,0.8879],[1.9997,0.2275],[19,0.2275],[19.9997,0.2]]}
          RGBA {type=MovieSceneScriptingFloatChannel, const [1,2]}
         ]
        }]
      }]
    }
   ]}
 }
 TeammateEliminated {
  binding_count=5
  length_seconds=1
  object=&self:TeammateEliminated
  sequence {bindings [
    EliminatedIcon {
     id=#5
     possessed_class=Image
     tracks [{
       type=MovieSceneByteTrack
       display_name=Visibility
       sections [{type=MovieSceneByteSection, end_frame=20, start_frame=0, channels [None {type=MovieSceneScriptingByteChannel, keys [[0,1],[2,4],[20,4]]}]}]
      }]
    }
    SquadMateBackground {
     id=#2
     display_name=CommonBorder_47
     possessed_class=CommonBorder
     tracks [{
       type=MovieSceneColorTrack
       display_name=Brush Color
       sections [{
         type=MovieSceneColorSection
         end_frame=20
         start_frame=0
         channels [
          Color.R {type=MovieSceneScriptingFloatChannel, keys [[0,0],[2,0.8148],[18,0.8148],[20,0]]}
          Color.G {type=MovieSceneScriptingFloatChannel, keys [[0,0],[2,0.0369],[18,0.0369],[20,0]]}
          Color.B {type=MovieSceneScriptingFloatChannel, keys [[0,0],[2,0.0307],[18,0.0307],[20,0]]}
          Color.A {type=MovieSceneScriptingFloatChannel, const [0.298,2]}
         ]
        }]
      }]
    }
    HitPointBarScaleBox {
     id=#6
     possessed_class=ScaleBox
     tracks [{
       type=MovieSceneByteTrack
       display_name=Visibility
       sections [{type=MovieSceneByteSection, end_frame=20, start_frame=0, channels [None {type=MovieSceneScriptingByteChannel, keys [[0,4],[2,2],[20,2]]}]}]
      }]
    }
    VerticalBox_42 {
     id=#4
     display_name=HitPointElimIconOverlay
     possessed_class=VerticalBox
     tracks [{
       type=MovieScene2DTransformTrack
       sections [{
         type=MovieScene2DTransformSection
         end_frame=20
         start_frame=0
         channels [
          Translation.X {type=MovieSceneScriptingFloatChannel, const [0,1]}
          Translation.Y {type=MovieSceneScriptingFloatChannel, const [0,1]}
          Angle {type=MovieSceneScriptingFloatChannel, const [0,1]}
          Scale.X {type=MovieSceneScriptingFloatChannel, thinned_from=8, keys [[0,1],[2.0003,1.25],[5,1],[8,1.25],[18,1.25],[20,1]]}
          Scale.Y {type=MovieSceneScriptingFloatChannel, keys [[0,1],[2.0003,1.25],[4.9997,1],[8,1.25],[18,1.25],[20,1]]}
          Shear.X {type=MovieSceneScriptingFloatChannel}
          Shear.Y {type=MovieSceneScriptingFloatChannel}
         ]
        }]
      }]
    }
    ShieldBar {
     id=#7
     possessed_class=AthenaSpectatorHitPointBar_C
     tracks [{
       type=MovieSceneByteTrack
       display_name=Visibility
       sections [{type=MovieSceneByteSection, end_frame=0, start_frame=0, channels [None {type=MovieSceneScriptingByteChannel, const [1,1]}]}]
      }]
    }
   ]}
 }
 ElimScale {
  binding_count=1
  length_seconds=0.2
  object=&self:ElimScale
  sequence {bindings [ElimCountSB {
     id=#8
     possessed_class=SizeBox
     tracks [{
       type=MovieScene2DTransformTrack
       sections [{
         type=MovieScene2DTransformSection
         end_frame=4
         start_frame=0
         channels [
          Translation.X {type=MovieSceneScriptingFloatChannel}
          Translation.Y {type=MovieSceneScriptingFloatChannel}
          Angle {type=MovieSceneScriptingFloatChannel}
          Scale.X {type=MovieSceneScriptingFloatChannel, keys [[0,1],[1,1.5],[3,1.5],[3.9997,1]]}
          Scale.Y {type=MovieSceneScriptingFloatChannel, keys [[0,1],[1,1.5],[3,1.5],[3.9997,1]]}
          Shear.X {type=MovieSceneScriptingFloatChannel}
          Shear.Y {type=MovieSceneScriptingFloatChannel}
         ]
        }]
      }]
    }]}
 }
]
graphs [
 EventGraph {nodes [
   "On Button Clicked (SimpleCommonButton)"!() -> (OutputDelegate, then, Button:Common Button Legacy Object Reference) {is=title K2Node_ComponentBoundEvent}
   > n1 = Cast To FortPlayerControllerSpectating!(execute, Object=GetOwningPlayer()) -> (then, CastFailed, AsFort Player Controller Spectating:Fort Player Controller Spectating Object Reference) {is=title K2Node_DynamicCast}
   n2 = GetPlayerState!(self:Athena Remote Player View Data Object Reference=GetPlayerData!(self:Athena Spectator Player Widget Base Object Reference) -> (ReturnValue:Athena Remote Player View Data Object Reference)) -> (ReturnValue:Fort Player State Athena Object Reference)
   > Branch(Condition=IsValidToFollow!(self:Fort Player Controller Spectating Object Reference=@n1, FPS:Fort Player State Object Reference=@n2) -> (ReturnValue:bool))
   > SetFollowedPlayer!(execute, self:Fort Player Controller Spectating Object Reference=@n1, NewPlayerToFollow:Fort Player State Object Reference=@n2) -> (then)
   ClearTextWidgets() {is=title K2Node_CustomEvent}
   > "SetText (Text)"(self=Get PlayerName() -> (PlayerName:Common Text Object Reference))
   > SetCurrentValue!(
    execute
    self:Common Numeric Text Block Object Reference=Get ElimCount() -> (ElimCount:Common Numeric Text Block Object Reference)
    NewValue:float=0
   ) -> (then)
   n3 = OnEliminatedStateChanged!() -> (OutputDelegate, then, bIsEliminated:bool=~) {is=title K2Node_CustomEvent}
   > n4 = Sequence()
   > n5 = Branch(Condition=@n3.bIsEliminated)
   n6 = Get bIsFollowedPlayer() -> (bIsFollowedPlayer:bool)
   > PlayAnimation(InAnimation=Select(
     Option 0:Widget Animation Object Reference=Get TeammateEliminated() -> (TeammateEliminated:Widget Animation Object Reference)
     Option 1:Widget Animation Object Reference=Get SpectatedEliminated() -> (SpectatedEliminated:Widget Animation Object Reference)
     Index=@n6
    ) -> (ReturnValue:Widget Animation Object Reference))
   > SetIsInteractionEnabled!(
    execute
    self:Common Button Base Object Reference=Get SimpleCommonButton() -> (SimpleCommonButton:Simple Common Button Object Reference)
    bInIsInteractionEnabled:bool=false
   ) -> (then)
   StopAllAnimations(execute=@n5.else, self:User Widget Object Reference)
   > Branch(Condition=@n6)
   > RefreshFollowedPlayerBackground!(
    execute
    self
    GradientColor1:Linear Color Structure=Get DesiredFollowedPlayerGradientColor1() -> (DesiredFollowedPlayerGradientColor1:Linear Color Structure)
    GradientColor2:Linear Color Structure=Get DesiredFollowedPlayerGradientColor2() -> (DesiredFollowedPlayerGradientColor2:Linear Color Structure)
    ElimBackgroundColor:Linear Color Structure=Get DesiredFollowedPlayerElimBackgroundColor() -> (DesiredFollowedPlayerElimBackgroundColor:Linear Color Structure)
   ) -> (then)
   Branch(execute=@n4.then_1, Condition=NOT Boolean(A=Get bIsFollowedPlayer()))
   > SetVisibility(
    self=Get HitPointBarScaleBox() -> (HitPointBarScaleBox:Scale Box Object Reference)
    InVisibility=Select(Option 0:ESlateVisibility Enum=SelfHitTestInvisible, Option 1:ESlateVisibility Enum=Hidden, Index=@n3.bIsEliminated) -> (ReturnValue:ESlateVisibility Enum)
   )
   > SetVisibility(
    self=Get EliminatedIcon() -> (EliminatedIcon:Image Object Reference)
    InVisibility=Select(Option 0=Collapsed, Option 1=SelfHitTestInvisible, Index=@n3.bIsEliminated)
   )
   > SetIsInteractionEnabled(self=Get SimpleCommonButton(), bInIsInteractionEnabled=true)
   n7 = UpdateFromMinigameTeamId!() -> (OutputDelegate, then, MinigameTeamId:Fort Minigame Team Identification Object Reference) {is=title K2Node_CustomEvent}
   > n8 = Sequence()
   > n9 = Is Valid(InputObject=@n7.MinigameTeamId)
   Set DesiredFollowedPlayerGradientColor1(
    execute=@n9.Is Not Valid
    DesiredFollowedPlayerGradientColor1:Linear Color Structure=Get FollowedPlayerNormalGradientColor1() -> (FollowedPlayerNormalGradientColor1:Linear Color Structure)
   ) -> (Output_Get:Linear Color Structure)
   > Set DesiredFollowedPlayerGradientColor2(DesiredFollowedPlayerGradientColor2:Linear Color Structure=Get FollowedPlayerNormalGradientColor2() -> (FollowedPlayerNormalGradientColor2:Linear Color Structure)) -> (Output_Get:Linear Color Structure)
   > Set DesiredFollowedPlayerElimBackgroundColor(DesiredFollowedPlayerElimBackgroundColor:Linear Color Structure=Get FollowedPlayerNormalElimBackgroundColor() -> (FollowedPlayerNormalElimBackgroundColor:Linear Color Structure)) -> (Output_Get:Linear Color Structure)
   > Set DesiredSquadMateBackgroundColor(DesiredSquadMateBackgroundColor:Linear Color Structure=(R=0,G=0,B=0,A=0.3)) -> (Output_Get:Linear Color Structure)
   Set DesiredFollowedPlayerGradientColor1(
    execute=@n9.Is Valid
    DesiredFollowedPlayerGradientColor1="To LinearColor (Color)"!(InColor:Color Structure=Get StartGradient!(self:Fort Minigame Team Identification Object Reference=@n7.MinigameTeamId) -> (StartGradient:Color Structure)) -> (ReturnValue:Linear Color Structure)
   )
   > Set DesiredFollowedPlayerGradientColor2(DesiredFollowedPlayerGradientColor2="To LinearColor (Color)"(InColor=Get EndGradient!(self:Fort Minigame Team Identification Object Reference=@n7.MinigameTeamId) -> (EndGradient:Color Structure)))
   n10 = "To LinearColor (Color)"(InColor=Get FirstColor!(self:Fort Minigame Team Identification Object Reference=@n7.MinigameTeamId) -> (FirstColor:Color Structure))
   > Set DesiredFollowedPlayerElimBackgroundColor(DesiredFollowedPlayerElimBackgroundColor=@n10)
   n11 = Break Linear Color!(LinearColor:"Linear Color Structure (by ref)"=@n10) -> (R:float=~, G:float=~, B:float=~, A:float=~) {is=title K2Node_BreakStruct}
   > Set!(
    execute
    DesiredSquadMateBackgroundColor_R:float=@n11
    DesiredSquadMateBackgroundColor_G:float=@n11.G
    DesiredSquadMateBackgroundColor_B:float=@n11.B
    DesiredSquadMateBackgroundColor_A:float=0.3
   ) -> (then, Output_Get:Linear Color Structure)
   SetBrushColor(
    execute=@n8.then_1
    self=Get SquadMateBackground() -> (SquadMateBackground:Common Border Object Reference)
    InBrushColor=Get DesiredSquadMateBackgroundColor() -> (DesiredSquadMateBackgroundColor:Linear Color Structure)
   )
   > RefreshFollowedPlayerBackground(
    GradientColor1=Get DesiredFollowedPlayerGradientColor1()
    GradientColor2=Get DesiredFollowedPlayerGradientColor2()
    ElimBackgroundColor=Get DesiredFollowedPlayerElimBackgroundColor()
   )
   n12 = Event PreConstruct()
   > Set DesiredFollowedPlayerGradientColor1(DesiredFollowedPlayerGradientColor1=Get FollowedPlayerNormalGradientColor1())
   > Set DesiredFollowedPlayerGradientColor2(DesiredFollowedPlayerGradientColor2=Get FollowedPlayerNormalGradientColor2())
   > Set DesiredFollowedPlayerElimBackgroundColor(DesiredFollowedPlayerElimBackgroundColor=Get FollowedPlayerNormalElimBackgroundColor())
   > Set DesiredSquadMateBackgroundColor(DesiredSquadMateBackgroundColor=Get BrushColor!(self:Border Object Reference=Get SquadMateBackground()) -> (BrushColor:Linear Color Structure))
   > UpdateStyle()
   > n13 = Branch(Condition=@n12.IsDesignTime)
   ClearTextWidgets(execute=@n13.else) {is=title K2Node_CallFunction}
   > SetIsInteractionEnabled(self=Get SimpleCommonButton(), bInIsInteractionEnabled=NOT Boolean(A=Get bIsFollowedPlayer()))
   n14 = Event OnDBNOStateChanged!() -> (OutputDelegate, then, bIsDBNO:bool) {is=title K2Node_Event}
   > UpdateDbnoIcon()
   > Branch(Condition=@n14.bIsDBNO)
   > PlayAnimation(InAnimation=Select(
     Option 0:Widget Animation Object Reference=Get TeammateDowned() -> (TeammateDowned:Widget Animation Object Reference)
     Option 1:Widget Animation Object Reference=Get SpectatedDowned() -> (SpectatedDowned:Widget Animation Object Reference)
     Index=Get bIsFollowedPlayer()
    ) -> (ReturnValue:Widget Animation Object Reference))
   n15 = Event OnDeadStateChanged!() -> (OutputDelegate, then, bIsDead:bool) {is=title K2Node_Event}
   > OnEliminatedStateChanged!(execute, self, bIsEliminated:bool=@n15.bIsDead) -> (then) {is=title K2Node_CallFunction}
   n16 = Event OnSquadIdVisibilityChanged!() -> (OutputDelegate, then, bVisible:bool) {is=title K2Node_Event}
   > SetVisibility(
    self=Get SquadIdSB() -> (SquadIdSB:Size Box Object Reference)
    InVisibility=Select(Option 0:ESlateVisibility Enum=Collapsed, Option 1:ESlateVisibility Enum=SelfHitTestInvisible, Index=@n16.bVisible) -> (ReturnValue:ESlateVisibility Enum)
   )
   n17 = Event Handle Playlist Extension Widget Created!() -> (OutputDelegate, then, Widget:User Widget Object Reference, SlotType:EUIExtensionSlot Enum=Invalid) {is=title K2Node_Event}
   > n18 = Switch on EUIExtensionSlot!(execute, Selection:EUIExtensionSlot Enum=@n17.SlotType) -> (
    Invalid:Exec
    Primary:Exec
    Secondary:Exec
    TopRightCorner:Exec
    GameInfoBox:Exec
    UpperCenter:Exec
    CrosshairRight:Exec
    MobileCrosshairRight:Exec
    UnderSquadInfo:Exec
    FullScreenMap:Exec
    BelowRespawnWidget:Exec
    BelowCompass:Exec
    UnderTeammateStatus:Exec
    ControllerBindingCallout:Exec
    AboveStormMessageSlot:Exec
    CustomMinigameCallouts:Exec
    UnderLocalPlayerInfo:Exec
    PlayerHealthbarOverlay:Exec
    UnderMinimapMessageContainer:Exec
    VehicleActionsSpacer:Exec
    Reticle:Exec
    PrioritizedContextualSlot:Exec
    RightOfTeammateStatus:Exec
    RightOfTeammateStatusBars:Exec
    TeammateStatusPortraitOverlay:Exec
    TeammateStatusPortraitUnderlay:Exec
    TeammateStatusPortraitReplacement:Exec
    TeammateStatusPortraitOverlay_IGM:Exec
    TeammateOverHealthbarOverlay:Exec
    TeammateOverShieldbarOverlay:Exec
    MobileHUDBottomRight:Exec
    InventoryScreenTab:Exec
    InventoryEquipSlot:Exec
    UnderneathTeammateStatusList:Exec
    FullScreenMapSquadEliminationsOverlay:Exec
    FullScreenMapStormInfo:Exec
    PrimaryQuickBarSlot:Exec
    MainMenuButtonListInZone:Exec
    MainMenuButtonListFrontEnd:Exec
    DynamicOverlayLobby:Exec
    PostMatchXPDetails_AccoladeList:Exec
    MobileHUDUnderMinimap:Exec
    SidebarPanels:Exec
    SpectatorRightOfPlayerStatus:Exec
    BehindMobileLocalPlayerInfo:Exec
    UnderMobileLocalPlayerInfo:Exec
    AboveMobileLocalPlayerInfo:Exec
    DynamicOverlayFrontEnd:Exec
    NPCStatusWidgetBarsPrefix:Exec
    NPCStatusWidgetAboveBars:Exec
    AbovePlayerIndicatorWidget:Exec
    FullScreenMapTab:Exec
    PostMatchScreenTab:Exec
    AboveBacchusEquippedItemReloadWidget:Exec
    UnderDefaultLobbyDetails:Exec
    DiscoveredAreaTextUnderlay:Exec
    LevelUpMobile:Exec
    ExpMobile:Exec
    InventoryGroupPanel:Exec
    InventoryItemInfoContents:Exec
    DiscoverInLobby:Exec
    InventoryItemInfoInspectPreview:Exec
    AboveQuestsMobileUI:Exec
    PostGamePlacementOverlayVictory:Exec
    TeammateStatusHealthShieldsReplacement:Exec
    GameInfoBelowMap:Exec
    LeftPostGameMobileOverlay:Exec
    LobbyAboveRankedDetails:Exec
    HUDPlayerActionAlertKillPlayerInfoExtension:Exec
    PostGamePlacementOverlayLosing:Exec
    AthenaInteractionIndicator:Exec
    HUDUnderReticle:Exec
    AthenaInventoryGroupMinorMedallionHeader:Exec
    AboveUpperCenterPrioritized:Exec
   ) {is=title K2Node_SwitchEnum}
   n19 = AddChildToOverlay!(
    execute=@n18.SpectatorRightOfPlayerStatus
    self:Overlay Object Reference=Get StatusExtensionOverlay() -> (StatusExtensionOverlay:Overlay Object Reference)
    Content:Widget Object Reference=@n17.Widget
   ) -> (then, ReturnValue:Overlay Slot Object Reference)
   > SetHorizontalAlignment(self=@n19, InHorizontalAlignment=HAlign_Fill)
   > SetVerticalAlignment(self=@n19, InVerticalAlignment=VAlign_Fill)
   n20 = Event OnPlayerNameChanged!() -> (OutputDelegate, then, PlayerName:String) {is=title K2Node_Event}
   > "SetText (Text)"(self=Get PlayerName(), InText="To Text (String)"(InString=@n20.PlayerName))
   n21 = Event OnSquadIdChanged!() -> (OutputDelegate, then, SquadId:Byte=0) {is=title K2Node_Event}
   > "SetText (Text)"(self=Get SquadId() -> (SquadId:Text Object Reference), InText="To Text (Byte)"!(Value:Byte=@n21.SquadId) -> (ReturnValue:Text))
   n22 = Event OnTeamColorChanged!() -> (OutputDelegate, then, TeamColor:Linear Color Structure) {is=title K2Node_Event}
   > SetBrushColor(self=Get SquadIdBorder() -> (SquadIdBorder:Common Border Object Reference), InBrushColor=@n22.TeamColor)
   n23 = Event OnPipModeChanged!() -> (OutputDelegate, then, bPipModeEnabled:bool) {is=title K2Node_Event}
   > Branch(Condition=NOT Boolean(A=Get bIsFollowedPlayer()))
   > SetVisibility(self=Get SquadIdBorder(), InVisibility=Select(Option 0=Collapsed, Option 1=Hidden, Index=@n23.bPipModeEnabled))
   n24 = Event OnPlayerDataChanged!() -> (OutputDelegate, then, InPlayerData:Athena Remote Player View Data Object Reference) {is=title K2Node_Event}
   > SetPlayerData!(
    execute
    self:Athena Spectator Hit Point Bar Base Object Reference=[
     Get ShieldBar() -> (ShieldBar:Athena Spectator Hit Point Bar Object Reference)
     Get HealthBar() -> (HealthBar:Athena Spectator Hit Point Bar Object Reference)
    ]
    InPlayerData:Athena Remote Player View Data Object Reference=@n24.InPlayerData
   ) -> (then)
   > StopAllAnimations()
   Event OnBeingRevivedStateChanged!() -> (OutputDelegate, then, bIsBeingRevived:bool) {is=title K2Node_Event}
   > UpdateDbnoIcon()
   n25 = Event OnKillScoreChanged!() -> (OutputDelegate, then, KillScore:int=0) {is=title K2Node_Event}
   > SetCurrentValue(self=Get ElimCount(), NewValue="To Float (Integer)"(InInt=@n25.KillScore))
   > Branch(Condition="integer > integer"(A=@n25.KillScore, B=0) {is=title K2Node_CallFunction})
   > PlayAnimation(InAnimation=Get ElimScale() -> (ElimScale:Widget Animation Object Reference))
   "Animation Finished (SpectatedEliminated)"!() -> (OutputDelegate:Delegate, then:Exec) {is=title K2Node_WidgetAnimationEvent}
   > RefreshFollowedPlayerBackground(
    GradientColor1=Get FollowedPlayerEliminatedGradientColor1() -> (FollowedPlayerEliminatedGradientColor1:Linear Color Structure)
    GradientColor2=Get FollowedPlayerEliminatedGradientColor2() -> (FollowedPlayerEliminatedGradientColor2:Linear Color Structure)
    ElimBackgroundColor=Get FollowedPlayerEliminatedElimBackgroundColor() -> (FollowedPlayerEliminatedElimBackgroundColor:Linear Color Structure)
   )
   "Animation Finished (SpectatedDowned)"!() -> (OutputDelegate:Delegate, then:Exec) {is=title K2Node_WidgetAnimationEvent}
   > RefreshFollowedPlayerBackground(
    GradientColor1=Get DesiredFollowedPlayerGradientColor1()
    GradientColor2=Get DesiredFollowedPlayerGradientColor2()
    ElimBackgroundColor=Get DesiredFollowedPlayerElimBackgroundColor()
   )
   "Animation Finished (TeammateDowned)"!() -> (OutputDelegate:Delegate, then:Exec) {is=title K2Node_WidgetAnimationEvent}
   > SetVisibility(self=Get ShieldBar(), InVisibility=Collapsed)
   "Animation Finished (TeammateEliminated)"!() -> (OutputDelegate:Delegate, then:Exec) {is=title K2Node_WidgetAnimationEvent}
   > SetVisibility(self=Get HitPointBarScaleBox(), InVisibility=Hidden)
   > SetVisibility(self=Get EliminatedIcon(), InVisibility=SelfHitTestInvisible)
   > SetBrushColor(self=Get SquadMateBackground(), InBrushColor=Get DesiredSquadMateBackgroundColor())
   OnEliminatedStateChanged(bIsEliminated=Event OnDisconnectedChanged!() -> (OutputDelegate, then, bIsDisconnected:bool) {is=title K2Node_Event}.bIsDisconnected)
  ]}
 RefreshFollowedPlayerBackground {nodes [
   RefreshFollowedPlayerBackground!() -> (then, GradientColor1:Linear Color Structure, GradientColor2:Linear Color Structure, ElimBackgroundColor:Linear Color Structure) {is=title K2Node_FunctionEntry}
   n1 = Get FollowedPlayerGradient() -> (FollowedPlayerGradient:Basic Gradient Fill Object Reference)
   > Set Fill!(
    execute
    self:Basic Gradient Fill Object Reference=@n1
    Color 1:Linear Color Structure=Get DesiredFollowedPlayerGradientColor1()
    Color 2:Linear Color Structure=Get DesiredFollowedPlayerGradientColor2()
    "Rotation (0_-_1)":double=Get Rotation!(self:Basic Gradient Fill Object Reference=@n1) -> (Rotation:double)
   ) -> (then) {is=title K2Node_CallFunction}
   > n2 = GetDynamicMaterial(self:Border Object Reference=Get FollowedPlayerElimBackground() -> (FollowedPlayerElimBackground:Common Border Object Reference))
   > SetVectorParameterValue(self=@n2, ParameterName=RGB, Value=Get DesiredFollowedPlayerElimBackgroundColor())
  ]}
 SetWidgetMode {nodes [
   n1 = SetWidgetMode!() -> (then, bIsFollowedPlayer:bool=~, bPipMode:bool=~) {is=title K2Node_FunctionEntry}
   > Set bIsFollowedPlayer(bIsFollowedPlayer:bool=@n1) -> (Output_Get:bool)
   > Set bPipMode(bPipMode:bool=@n1.bPipMode) -> (Output_Get:bool)
   > UpdateStyle()
   > SetIsInteractionEnabled(self=Get SimpleCommonButton(), bInIsInteractionEnabled=NOT Boolean(A=Get bIsFollowedPlayer()))
  ]}
 UpdateDbnoIcon {nodes [
   UpdateDbnoIcon() {is=title K2Node_FunctionEntry}
   n1 = GetPlayerData()
   n2 = GetIsBeingRevived!(self:Athena Remote Player View Data Object Reference=@n1) -> (ReturnValue:bool)
   n3 = Get DbnoStateImage() -> (DbnoStateImage:Image Object Reference)
   > SetBrushFromSoftMaterial!(
    execute
    self:Image Object Reference=@n3
    SoftMaterial:Material Interface Soft Object Reference=Select(
     Option 0:Material Interface Soft Object Reference=Get DbnoIcon() -> (DbnoIcon:Material Soft Object Reference)
     Option 1:Material Interface Soft Object Reference=Get ReviveIcon() -> (ReviveIcon:Material Soft Object Reference)
     Index=@n2
    ) -> (ReturnValue:Material Interface Soft Object Reference)
   ) -> (then)
   n4 = GetIsDBNO!(self:Athena Remote Player View Data Object Reference=@n1) -> (ReturnValue:bool)
   > SetVisibility(
    self=@n3
    InVisibility=Select(Option 0:ESlateVisibility Enum=Collapsed, Option 1:ESlateVisibility Enum=SelfHitTestInvisible, Index=OR Boolean(A=@n2, B=@n4)) -> (ReturnValue:ESlateVisibility Enum)
   )
  ]}
 UpdateGradientFill {nodes [
   n1 = UpdateGradientFill!() -> (then, Target:Basic Gradient Fill Object Reference) {is=title K2Node_FunctionEntry}
   > Set Fill(
    self=@n1
    Color 1=Get Color 1!(self:Basic Gradient Fill Object Reference=@n1) -> (Color 1:Linear Color Structure)
    Color 2=Get Color 2!(self:Basic Gradient Fill Object Reference=@n1) -> (Color 2:Linear Color Structure)
    "Rotation (0_-_1)"=Get Rotation(self=@n1)
   )
  ]}
 UpdateStyle {nodes [
   UpdateStyle() {is=title K2Node_FunctionEntry}
   n1 = Get bIsFollowedPlayer()
   > SetActiveWidgetIndex(
    self=Get BackgroundSwitcher() -> (BackgroundSwitcher:Common Widget Switcher Legacy Object Reference)
    Index=Select(Option 0:int=1, Option 1:int=0, Index=@n1) -> (ReturnValue:int)
   )
   > SetVisibility(
    self=Get HitPointElimIconOverlay() -> (HitPointElimIconOverlay:Overlay Object Reference)
    InVisibility=Select(Option 0:ESlateVisibility Enum=SelfHitTestInvisible, Option 1:ESlateVisibility Enum=Collapsed, Index=@n1) -> (ReturnValue:ESlateVisibility Enum)
   )
   n2 = Get bPipMode() -> (bPipMode:bool)
   n3 = Get FollowedPlayerNameFont_Pip() -> (FollowedPlayerNameFont_Pip:Slate Font Info Structure)
   n4 = Get FollowedPlayerNameFont_NoPip() -> (FollowedPlayerNameFont_NoPip:Slate Font Info Structure)
   > SetFont(
    self=Get PlayerName()
    InFontInfo=Select(
     Option 0:Slate Font Info Structure=Get SquadmatePlayerNameFont() -> (SquadmatePlayerNameFont:Slate Font Info Structure)
     Option 1:Slate Font Info Structure=Select(Option 0=@n4, Option 1=@n3, Index=@n2)
     Index=@n1
    ) -> (ReturnValue:Slate Font Info Structure)
   )
   > SetVisibility(
    self=Get SquadIdBorder()
    InVisibility=Select(Option 0:ESlateVisibility Enum=Collapsed, Option 1:ESlateVisibility Enum=SelfHitTestInvisible, Index=@n1) -> (ReturnValue:ESlateVisibility Enum)
   )
   n5 = AND Boolean(A=NOT Boolean(A=Get bPipMode()), B=@n1)
   > SetFont(
    self=[Get ElimCount(), Get SquadId()]
    InFontInfo=Select(
     Option 0:Slate Font Info Structure=Get KillsFont_Pip() -> (KillsFont_Pip:Slate Font Info Structure)
     Option 1:Slate Font Info Structure=Get KillsFont_NoPip() -> (KillsFont_NoPip:Slate Font Info Structure)
     Index=@n5
    ) -> (ReturnValue:Slate Font Info Structure)
   )
   > SetMinDesiredWidth!(
    execute
    self:Size Box Object Reference=Get ElimBackgroundSizeBox() -> (ElimBackgroundSizeBox:Size Box Object Reference)
    InMinDesiredWidth:float=Select(Option 0:double=105, Option 1:double=140, Index=@n5) -> (ReturnValue:double)
   ) -> (then)
   > SetMinDesiredWidth(self=Get ElimSizeBox() -> (ElimSizeBox:Size Box Object Reference), InMinDesiredWidth=Select(Option 0=78, Option 1=112, Index=@n5))
   > SetDesiredSizeOverride!(
    execute
    self:Image Object Reference=Get ElimIcon() -> (ElimIcon:Image Object Reference)
    DesiredSize:Vector 2D Structure=Select(Option 0:Vector 2D Structure=(X=26,Y=26), Option 1:Vector 2D Structure=(X=35,Y=35), Index=@n5) -> (ReturnValue:Vector 2D Structure)
   ) -> (then)
   > SetPadding!(
    execute
    self:Horizontal Box Slot Object Reference=SlotAsHorizontalBoxSlot!(Widget:Widget Object Reference=Get ElimCountSB() -> (ElimCountSB:Size Box Object Reference)) -> (ReturnValue:Horizontal Box Slot Object Reference)
    InPadding_Left:float=0
    InPadding_Top:float=0
    InPadding_Right:float=8
    InPadding_Bottom:float=Select(Option 0:double=0, Option 1:double=-0.75, Index=@n5) -> (ReturnValue:double)
   ) -> (then)
   > SetPadding(
    self=SlotAsHorizontalBoxSlot(Widget=Get ElimSizeBox())
    InPadding_Left=Select(Option 0=0, Option 1=10, Index=Get bPipMode())
    InPadding_Top=0
    InPadding_Right=0
    InPadding_Bottom=0
   )
   > SetPadding(
    self=SlotAsHorizontalBoxSlot(Widget=Get ElimIcon())
    InPadding_Left=Select(Option 0=10, Option 1=3.5, Index=@n5)
    InPadding_Top=0
    InPadding_Right=Select(Option 0=0, Option 1=1.5, Index=@n5)
    InPadding_Bottom=0
   )
   > SetPadding(
    self=SlotAsHorizontalBoxSlot(Widget=Get PlayerNameSizeBox() -> (PlayerNameSizeBox:Size Box Object Reference))
    InPadding_Left=10
    InPadding_Top=0
    InPadding_Right=Select(Option 0=5, Option 1=26, Index=@n5)
    InPadding_Bottom=0
   )
  ]}
]
variables [
 bIsFollowedPlayer {type=boolean, default=true}
 bPipMode {type=boolean}
 DbnoIcon {type=Material, default=/Game/Athena/HUD/Art/M_Dying_Indicator, class=~, graph=~}
 DesiredFollowedPlayerElimBackgroundColor {type=LinearColor, default=<rgba 0.1095 0.1095 0.1095 1>}
 DesiredFollowedPlayerGradientColor1 {type=LinearColor, default=<rgba 0.1095 0.1095 0.1095 1>}
 DesiredFollowedPlayerGradientColor2 {type=LinearColor, default=<rgba 0.1095 0.1095 0.1095 1>}
 DesiredSquadMateBackgroundColor {type=LinearColor, default=<rgba 0.1095 0.1095 0.1095 1>}
 FollowedPlayerEliminatedElimBackgroundColor {type=LinearColor, default=<rgba 0.2 0.2 0.2 1>}
 FollowedPlayerEliminatedGradientColor1 {type=LinearColor, default=<rgba 0.1095 0.1095 0.1095 1>}
 FollowedPlayerEliminatedGradientColor2 {type=LinearColor, default=<rgba 0.0497 0.0497 0.0497 1>}
 FollowedPlayerNameFont_NoPip {
  type=SlateFontInfo
  default {
   bForceMonospaced=false
   bMaterialIsStencil=false
   fontMaterial=null
   fontObject=&BurbankSmall
   letterSpacing=0
   monospacedWidth=1
   size=26
   skewAmount=0
   typefaceFontName=Black
   outlineSettings {
    bApplyOutlineToDropShadows=false
    bMiteredCorners=false
    bSeparateFillAlpha=false
    outlineMaterial=null
    outlineSize=0
    outlineColor=<rgba 0 0 0 1>
   }
  }
 }
 FollowedPlayerNameFont_Pip {
  type=SlateFontInfo
  default {
   bForceMonospaced=false
   bMaterialIsStencil=false
   fontMaterial=null
   fontObject=&BurbankSmall
   letterSpacing=0
   monospacedWidth=1
   size=20
   skewAmount=0
   typefaceFontName=Bold
   outlineSettings {
    bApplyOutlineToDropShadows=false
    bMiteredCorners=false
    bSeparateFillAlpha=false
    outlineMaterial=null
    outlineSize=0
    outlineColor=<rgba 0 0 0 1>
   }
  }
 }
 FollowedPlayerNormalElimBackgroundColor {type=LinearColor, default=<rgba 0.0423 0.3325 0.8879 1>}
 FollowedPlayerNormalGradientColor1 {type=LinearColor, default=<rgba 0.0044 0.1022 0.4793 1>}
 FollowedPlayerNormalGradientColor2 {type=LinearColor, default=<rgba 0.0513 0.1529 0.5647 1>}
 KillsFont_NoPip {
  type=SlateFontInfo
  default {
   bForceMonospaced=false
   bMaterialIsStencil=false
   fontMaterial=null
   fontObject=&BurbankSmall
   letterSpacing=0
   monospacedWidth=1
   size=30
   skewAmount=0
   typefaceFontName=Black
   outlineSettings {
    bApplyOutlineToDropShadows=false
    bMiteredCorners=false
    bSeparateFillAlpha=false
    outlineMaterial=null
    outlineSize=0
    outlineColor=<rgba 0 0 0 1>
   }
  }
 }
 KillsFont_Pip {
  type=SlateFontInfo
  default {
   bForceMonospaced=false
   bMaterialIsStencil=false
   fontMaterial=null
   fontObject=&BurbankSmall
   letterSpacing=0
   monospacedWidth=1
   size=24
   skewAmount=0
   typefaceFontName=Black
   outlineSettings {
    bApplyOutlineToDropShadows=false
    bMiteredCorners=false
    bSeparateFillAlpha=false
    outlineMaterial=null
    outlineSize=0
    outlineColor=<rgba 0 0 0 1>
   }
  }
 }
 ReviveIcon {type=Material, default=/Game/Athena/HUD/Art/M_Healing_Indicator, class=~, graph=~}
 SquadmatePlayerNameFont {
  type=SlateFontInfo
  default {
   bForceMonospaced=false
   bMaterialIsStencil=false
   fontMaterial=null
   fontObject=&BurbankSmall
   letterSpacing=0
   monospacedWidth=1
   size=16
   skewAmount=0
   typefaceFontName=Bold
   outlineSettings {
    bApplyOutlineToDropShadows=false
    bMiteredCorners=false
    bSeparateFillAlpha=false
    outlineMaterial=null
    outlineSize=0
    outlineColor=<rgba 0 0 0 1>
   }
  }
 }
]
widgets {
 info {inheritedWidgetCount=0, parentClass=/Script/FortniteUI.AthenaSpectatorPlayerWidgetBase, rootWidgetClass=&SizeBox, widgetCount=29}
 widgets [
  SizeBox {widgetClassPath=&SizeBox, bIsVariable=true, widget=SizeBox, properties {bOverride_HeightOverride=true, heightOverride=50}}
  Overlay_1 {widgetClassPath=&Overlay, parent=SizeBox, widget=Overlay_1, slot {class=SizeBoxSlot}}
  BackgroundSwitcher {
   widgetClassPath=&CommonWidgetSwitcherLegacy
   bIsVariable=true
   parent=Overlay_1
   widget=BackgroundSwitcher
   properties {bOutroPanelBelow=false, bWidgetActivationEnabled=false, transitionDuration=0}
   slot {class=OverlaySlot, properties {horizontalAlignment=HAlign_Fill, verticalAlignment=VAlign_Fill}}
  }
  FollowedPlayerOverlay {widgetClassPath=&Overlay, parent=BackgroundSwitcher, widget=FollowedPlayerOverlay, slot {class=WidgetSwitcherSlot}}
  FollowedPlayerGradient {
   widgetClassPath=/Game/UI/Foundation/Stats/BasicGradientFill
   bIsVariable=true
   parent=FollowedPlayerOverlay
   widget=FollowedPlayerGradient
   properties {rotation=0.3, brush Size=<0.254 0.1983>, color 1=<rgba 0.0044 0.1022 0.4793 1>, color 2=<rgba 0.0523 0.1526 0.5647 1>}
   slot {class=OverlaySlot, properties {horizontalAlignment=HAlign_Fill, verticalAlignment=VAlign_Fill}}
  }
  ElimBackgroundSizeBox {
   widgetClassPath=&SizeBox
   bIsVariable=true
   parent=FollowedPlayerOverlay
   widget=ElimBackgroundSizeBox
   properties {bOverride_MinDesiredWidth=true, minDesiredWidth=78}
   slot {class=OverlaySlot, properties {horizontalAlignment=HAlign_Right, verticalAlignment=VAlign_Fill}}
  }
  FollowedPlayerElimBackground {
   widgetClassPath=&CommonBorder
   bIsVariable=true
   parent=ElimBackgroundSizeBox
   widget=FollowedPlayerElimBackground
   properties {
    background {
     bIsDynamicallyLoaded=false
     drawAs=Image
     imageType=NoImage
     mirroring=NoMirror
     resourceName=None
     resourceObject=/Game/Athena/HUD/Spectator/Art/MI_BroadcastSquadElimBG
     tiling=NoTile
     imageSize=<32 32>
     margin=<ltrb 0 0 0 0>
     outlineSettings {
      bUseBrushTransparency=false
      roundingType=HalfHeightRadius
      width=0
      color {colorUseRule=UseColor_Specified, specifiedColor=<rgba 0 0 0 0>}
      cornerRadii=<0 0 0 0>
     }
     tintColor {colorUseRule=UseColor_Specified, specifiedColor=<rgba 1 1 1 1>}
     uVRegion {bIsValid=false, max=<0 0>, min=<0 0>}
    }
    padding=<ltrb 0 0 0 0>
   }
   slot {class=SizeBoxSlot}
  }
  SquadMateBackground {
   widgetClassPath=&CommonBorder
   bIsVariable=true
   parent=BackgroundSwitcher
   widget=SquadMateBackground
   properties {
    background {
     bIsDynamicallyLoaded=false
     drawAs=Box
     imageType=NoImage
     mirroring=NoMirror
     resourceName=None
     resourceObject=null
     tiling=NoTile
     imageSize=<32 32>
     margin=<ltrb 0 0 0 0>
     outlineSettings {
      bUseBrushTransparency=false
      roundingType=HalfHeightRadius
      width=0
      color {colorUseRule=UseColor_Specified, specifiedColor=<rgba 0 0 0 0>}
      cornerRadii=<0 0 0 0>
     }
     tintColor {colorUseRule=UseColor_Specified, specifiedColor=<rgba 1 1 1 1>}
     uVRegion {bIsValid=false, max=<0 0>, min=<0 0>}
    }
    brushColor=<rgba 0 0 0 0.3>
    padding=<ltrb 0 0 0 0>
   }
   slot {class=WidgetSwitcherSlot}
  }
  HorizontalBox_87 {
   widgetClassPath=&HorizontalBox
   parent=Overlay_1
   widget=HorizontalBox_87
   slot {class=OverlaySlot, properties {horizontalAlignment=HAlign_Fill, verticalAlignment=VAlign_Fill}}
  }
  SquadIdSB {
   widgetClassPath=&SizeBox
   bIsVariable=true
   parent=HorizontalBox_87
   widget=SquadIdSB
   properties {bOverride_WidthOverride=true, visibility=Collapsed, widthOverride=50}
   slot {class=HorizontalBoxSlot}
  }
  SquadIdBorder {
   widgetClassPath=&CommonBorder
   bIsVariable=true
   parent=SquadIdSB
   widget=SquadIdBorder
   properties {
    horizontalAlignment=HAlign_Center
    verticalAlignment=VAlign_Center
    background {
     bIsDynamicallyLoaded=false
     drawAs=Box
     imageType=NoImage
     mirroring=NoMirror
     resourceName=None
     resourceObject=null
     tiling=NoTile
     imageSize=<32 32>
     margin=<ltrb 0 0 0 0>
     outlineSettings {
      bUseBrushTransparency=false
      roundingType=HalfHeightRadius
      width=0
      color {colorUseRule=UseColor_Specified, specifiedColor=<rgba 0 0 0 0>}
      cornerRadii=<0 0 0 0>
     }
     tintColor {colorUseRule=UseColor_Specified, specifiedColor=<rgba 1 1 1 1>}
     uVRegion {bIsValid=false, max=<0 0>, min=<0 0>}
    }
    brushColor=<rgba 1 1 0 1>
    padding=<ltrb 0 0 0 0>
   }
   slot {class=SizeBoxSlot}
  }
  SquadId {
   widgetClassPath=&TextBlock
   bIsVariable=true
   parent=SquadIdBorder
   widget=SquadId
   properties {
    text=88
    visibility=SelfHitTestInvisible
    colorAndOpacity {colorUseRule=UseColor_Specified, specifiedColor=<rgba 0 0 0 0.4>}
    font {
     bForceMonospaced=false
     bMaterialIsStencil=false
     fontMaterial=null
     fontObject=&BurbankSmall
     letterSpacing=0
     monospacedWidth=1
     size=24
     skewAmount=0
     typefaceFontName=Black
     outlineSettings {
      bApplyOutlineToDropShadows=false
      bMiteredCorners=false
      bSeparateFillAlpha=false
      outlineMaterial=null
      outlineSize=0
      outlineColor=<rgba 0 0 0 1>
     }
    }
   }
   slot {class=BorderSlot, properties {horizontalAlignment=HAlign_Center, verticalAlignment=VAlign_Center, padding=<ltrb 0 0 0 0>}}
  }
  PlayerNameSizeBox {
   widgetClassPath=&SizeBox
   bIsVariable=true
   parent=HorizontalBox_87
   widget=PlayerNameSizeBox
   properties {bOverride_MaxDesiredWidth=true, maxDesiredWidth=200}
   slot {class=HorizontalBoxSlot, properties {verticalAlignment=VAlign_Center, padding=<ltrb 10 0 5 0>, size {sizeRule=Fill, value=1}}}
  }
  PlayerName {
   widgetClassPath=&CommonTextBlock
   bIsVariable=true
   parent=PlayerNameSizeBox
   widget=PlayerName
   properties {
    clipping=ClipToBounds
    text=Player Name
    font {
     bForceMonospaced=false
     bMaterialIsStencil=false
     fontMaterial=null
     fontObject=&NotoSans
     letterSpacing=0
     monospacedWidth=1
     size=24
     skewAmount=0
     typefaceFontName=NotoSans-Regular
     outlineSettings {
      bApplyOutlineToDropShadows=false
      bMiteredCorners=false
      bSeparateFillAlpha=false
      outlineMaterial=null
      outlineSize=0
      outlineColor=<rgba 0 0 0 1>
     }
    }
    shadowOffset=<0 0>
   }
   slot {class=SizeBoxSlot, properties {padding=<ltrb 0 0 0 1>}}
  }
  HitPointElimIconOverlay {
   widgetClassPath=&Overlay
   bIsVariable=true
   parent=HorizontalBox_87
   widget=HitPointElimIconOverlay
   slot {class=HorizontalBoxSlot, properties {size {sizeRule=Fill, value=0.3}}}
  }
  HitPointBarScaleBox {
   widgetClassPath=&ScaleBox
   bIsVariable=true
   parent=HitPointElimIconOverlay
   widget=HitPointBarScaleBox
   properties {stretch=Fill, userSpecifiedScale=0.7}
   slot {class=OverlaySlot, properties {horizontalAlignment=HAlign_Fill, verticalAlignment=VAlign_Fill, padding=<ltrb 0 12 0 12>}}
  }
  VerticalBox_42 {
   widgetClassPath=&VerticalBox
   parent=HitPointBarScaleBox
   widget=VerticalBox_42
   slot {class=ScaleBoxSlot, properties {horizontalAlignment=HAlign_Fill, verticalAlignment=VAlign_Fill}}
  }
  ShieldBar {
   widgetClassPath=/Game/Athena/HUD/Spectator/AthenaSpectatorHitPointBar
   bIsVariable=true
   parent=VerticalBox_42
   widget=ShieldBar
   properties {barType=Shield, hideShieldOnDbno=false}
   slot {class=VerticalBoxSlot, properties {padding=<ltrb 0 0 0 4>, size {sizeRule=Fill, value=1}}}
  }
  HealthBar {
   widgetClassPath=&AthenaSpectatorHitPointBar
   bIsVariable=true
   parent=VerticalBox_42
   widget=HealthBar
   slot {class=VerticalBoxSlot, properties {size {sizeRule=Fill, value=1}}}
  }
  EliminatedIcon {
   widgetClassPath=&Image
   bIsVariable=true
   parent=HitPointElimIconOverlay
   widget=EliminatedIcon
   properties {
    visibility=Collapsed
    brush {
     bIsDynamicallyLoaded=false
     drawAs=Image
     imageType=NoImage
     mirroring=NoMirror
     resourceName=None
     resourceObject=/Game/Athena/HUD/Art/squadArt/squad_Eliminated
     tiling=NoTile
     imageSize=<32 32>
     margin=<ltrb 0 0 0 0>
     outlineSettings {
      bUseBrushTransparency=false
      roundingType=HalfHeightRadius
      width=0
      color {colorUseRule=UseColor_Specified, specifiedColor=<rgba 0 0 0 0>}
      cornerRadii=<0 0 0 0>
     }
     tintColor {colorUseRule=UseColor_Specified, specifiedColor=<rgba 1 1 1 1>}
     uVRegion {bIsValid=false, max=<0 0>, min=<0 0>}
    }
    colorAndOpacity=<rgba 1 0 0 1>
   }
   slot {class=OverlaySlot, properties {horizontalAlignment=HAlign_Center, verticalAlignment=VAlign_Center}}
  }
  ElimSizeBox {
   widgetClassPath=&SizeBox
   bIsVariable=true
   parent=HorizontalBox_87
   widget=ElimSizeBox
   properties {minDesiredWidth=95}
   slot {class=HorizontalBoxSlot}
  }
  ElimHB {widgetClassPath=&HorizontalBox, parent=ElimSizeBox, widget=ElimHB, slot {class=SizeBoxSlot}}
  ElimIcon {
   widgetClassPath=&Image
   bIsVariable=true
   parent=ElimHB
   widget=ElimIcon
   properties {
    visibility=SelfHitTestInvisible
    brush {
     bIsDynamicallyLoaded=false
     drawAs=Image
     imageType=NoImage
     mirroring=NoMirror
     resourceName=None
     resourceObject=/Game/Athena/HUD/Spectator/Art/T-Icon-Elim-L
     tiling=NoTile
     imageSize=<26 26>
     margin=<ltrb 0 0 0 0>
     outlineSettings {
      bUseBrushTransparency=false
      roundingType=HalfHeightRadius
      width=0
      color {colorUseRule=UseColor_Specified, specifiedColor=<rgba 0 0 0 0>}
      cornerRadii=<0 0 0 0>
     }
     tintColor {colorUseRule=UseColor_Specified, specifiedColor=<rgba 1 1 1 1>}
     uVRegion {bIsValid=false, max=<0 0>, min=<0 0>}
    }
   }
   slot {class=HorizontalBoxSlot, properties {verticalAlignment=VAlign_Center, padding=<ltrb 2 0 0 0>}}
  }
  ElimCountSB {
   widgetClassPath=&SizeBox
   bIsVariable=true
   parent=ElimHB
   widget=ElimCountSB
   properties {bOverride_MinDesiredWidth=true, minDesiredWidth=52}
   slot {class=HorizontalBoxSlot, properties {verticalAlignment=VAlign_Center, padding=<ltrb 0 0 8 0>}}
  }
  ElimCount {
   widgetClassPath=/Script/CommonUI.CommonNumericTextBlock
   bIsVariable=true
   parent=ElimCountSB
   widget=ElimCount
   properties {
    justification=Center
    text=Text Block
    font {
     bForceMonospaced=false
     bMaterialIsStencil=false
     fontMaterial=null
     fontObject=&BurbankSmall
     letterSpacing=0
     monospacedWidth=1
     size=24
     skewAmount=0
     typefaceFontName=Black
     outlineSettings {
      bApplyOutlineToDropShadows=false
      bMiteredCorners=false
      bSeparateFillAlpha=false
      outlineMaterial=null
      outlineSize=0
      outlineColor=<rgba 0 0 0 1>
     }
    }
    shadowOffset=<0 0>
   }
   slot {class=SizeBoxSlot, properties {horizontalAlignment=HAlign_Center, verticalAlignment=VAlign_Center}}
  }
  SimpleCommonButton {
   widgetClassPath=/Game/Athena/UI/Matchmaking/SimpleCommonButton
   bIsVariable=true
   parent=Overlay_1
   widget=SimpleCommonButton
   properties {
    bApplyAlphaOnDisable=false
    bDisplayInputActionWhenNotInteractable=false
    bShouldUseFallbackDefaultInputAction=false
    onButtonBaseClicked=(&self:WidgetTree.SimpleCommonButton.HandleOnButtonClicked)
    onButtonBaseDoubleClicked=(&self:WidgetTree.SimpleCommonButton.HandleOnButtonDoubleClicked)
    onButtonBaseHovered=(&self:WidgetTree.SimpleCommonButton.HandleOnButtonHovered)
    onButtonBaseUnhovered=(&self:WidgetTree.SimpleCommonButton.HandleOnButtonUnhovered)
    onSelectedChangedBase=(&self:WidgetTree.SimpleCommonButton.HandleOnSelectedChanged)
    style=/FortUILibrary/Foundation/Buttons/ButtonStyles/Athena/ButtonStyle_TransparentList.ButtonStyle_TransparentList_C
   }
   slot {class=OverlaySlot, properties {horizontalAlignment=HAlign_Fill, verticalAlignment=VAlign_Fill}}
  }
  HorizontalBox_0 {
   widgetClassPath=&HorizontalBox
   parent=Overlay_1
   widget=HorizontalBox_0
   slot {class=OverlaySlot, properties {horizontalAlignment=HAlign_Right, verticalAlignment=VAlign_Center}}
  }
  DbnoStateImage {
   widgetClassPath=&Image
   bIsVariable=true
   parent=HorizontalBox_0
   widget=DbnoStateImage
   properties {
    visibility=Collapsed
    brush {
     bIsDynamicallyLoaded=false
     drawAs=Image
     imageType=NoImage
     mirroring=NoMirror
     resourceName=None
     resourceObject=&M_Dying_Indicator
     tiling=NoTile
     imageSize=<42 42>
     margin=<ltrb 0 0 0 0>
     outlineSettings {
      bUseBrushTransparency=false
      roundingType=HalfHeightRadius
      width=0
      color {colorUseRule=UseColor_Specified, specifiedColor=<rgba 0 0 0 0>}
      cornerRadii=<0 0 0 0>
     }
     tintColor {colorUseRule=UseColor_Specified, specifiedColor=<rgba 1 1 1 1>}
     uVRegion {bIsValid=false, max=<0 0>, min=<0 0>}
    }
    renderTransform {angle=0, scale=<1 1>, shear=<0 0>, translation=<47 0>}
   }
   slot {class=HorizontalBoxSlot}
  }
  StatusExtensionOverlay {
   widgetClassPath=&Overlay
   bIsVariable=true
   parent=HorizontalBox_0
   widget=StatusExtensionOverlay
   properties {renderTransform {angle=0, scale=<1 1>, shear=<0 0>, translation=<47 0>}}
   slot {class=HorizontalBoxSlot, properties {horizontalAlignment=HAlign_Left, verticalAlignment=VAlign_Top}}
  }
 ]
}
